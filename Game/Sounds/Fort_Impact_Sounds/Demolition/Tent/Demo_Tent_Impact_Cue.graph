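# Demolition impact sound cue for tents, played as a small-object hit in the game's impact sound class.
# A Random node picks one of three tent impact waves with equal weight, and a Modulator varies its pitch slightly on each play.
# It uses small-object demolition hit attenuation and a reduced volume.
# A shared object-impact concurrency set limits simultaneous impacts.
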SoundCue /Game/Sounds/Fort_Impact_Sounds/Demolition/Tent/Demo_Tent_Impact_Cue
root=Modulator_2
nodes [
 {id=Modulator_2, children [Random_8], properties {pitchMax=1.07, pitchMin=0.92}}
 {id=Random_8, children [WavePlayer_53, WavePlayer_58, WavePlayer_59], properties {weights [1,1,1]}}
 {id=WavePlayer_53, properties {soundWaveAssetPtr=&/Tent_Impact01}}
 {id=WavePlayer_58, properties {soundWaveAssetPtr=&/Tent_Impact02}}
 {id=WavePlayer_59, properties {soundWaveAssetPtr=&/Tent_Impact03}}
]
properties {
 attenuationSettings=/Game/Sounds/Attenuation/Fort_Demolition_Hit_Small_Sized_Object_Att
 maxDistance=4000
 soundClassObject=&Fort_Impacts
 volumeMultiplier=0.68
 concurrencyOverrides {resolutionRule=PreventNew}
 concurrencySet [&CONC_Object_Impacts_FourMax]
}
